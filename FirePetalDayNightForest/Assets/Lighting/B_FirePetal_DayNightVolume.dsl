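; Day/night volume for the Fire Petal forest that changes local time of day when a global time-of-day change event is received, gated by hotfix flags.
; The authority listens for the event, sets bIsActive, then sphere-overlaps actors around the capsule, with optional tag checks, to execute a gameplay cue on player controllers.
; OnRep_bIsActive plays a cosmetic timeline driving the DaySequence modifier's cycle time, forcing smooth blending until it finishes.
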(event Custom|ActivateDaytime
  (Networking|CanExecuteCosmeticEvents
    (:True
      (Variables|/Script/DaySequence.DaySequenceModifierComponent|SetbForceSmoothBlending true (Variables|Default|GetDaySequenceModifier))
      (bind (_direction _timeofdaychange) (Timeline|Timeline PlayFromStart)
        (:Update
          (Variables|/Script/DaySequence.DaySequenceModifierComponent|SetDayNightCycleTime _timeofdaychange (Variables|Default|GetDaySequenceModifier)))
        (:Finished
          (Variables|/Script/DaySequence.DaySequenceModifierComponent|SetbForceSmoothBlending false (Variables|Default|GetDaySequenceModifier)))))))

(event EventBeginPlay
  (|CallParentFunction)
  (if (ScalableFloat|GetValueasBool (Variables|Hotfix|GetHF_EnableDayNightVolume))
    (Networking|SwitchHasAuthority
      (:Authority
        (if (ScalableFloat|GetValueasBool (Variables|Hotfix|GetHF_EnableDayNightTransition))
          (GameplayEventRouter|StartListeningToEvent (GameplayEventRouter|GetGlobalEventRouter) "/Game/Conversation/Services/Effects/OnTimeOfDayChangeRequested.OnTimeOfDayChangeRequested"
            (:OnEventReceived
              (Variables|Default|SetbIsActive true)
              (CallFunction|OnBroadcastEventReceived))))))
    (else
      ; Disable Modifier if disabled via hotfix
      (Variables|Default|SetbDaySequenceModifierEnabled false))))

(event Custom|OnBroadcastEventReceived
  (bind (_outactors _returnvalue) (Collision|SphereOverlapActors (Transformation|GetWorldLocation (Variables|Default|GetCapsule1)) (Components|Capsule|GetScaledCapsuleHalfHeight (Variables|Default|GetCapsule1)) (Variables|OverlappingActorSettings|GetoverlapObjectTypes) (Variables|OverlappingActorSettings|GetactorClassReceivingEffects)))
  (bind _outputgameplaytagscontainer (CallFunction|CreateTagContainer))
  (if _returnvalue
    (for _array_element _outactors
      (if (Variables|OverlappingActorSettings|GetbCheckActorTag)
        (bind _returnvalue_2 (GameplayTags|HasAllMatchingGameplayTags(Message) _array_element _outputgameplaytagscontainer))
        (if _returnvalue_2
          _)
        (else
          _))
      (GameplayCue|ExecuteGameplayCue (Fortnite|GetFortPlayerControllerfromActor _array_element) (Variables|EnvironmentFX|GetgameplayCueTag) (Class|AthenaFunctionLibrary|MakeSimpleFortEffectContext self)))))

(fn createTagContainer
  (bind _gameplaytags (LocalVariables|Get_gameplayTags))
  (Utilities|Array|AddUnique _gameplaytags (Variables|OverlappingActorSettings|GetactorFilteringTag))
  (Utilities|Array|AddUnique _gameplaytags (Variables|OverlappingActorSettings|GetpoiTag))
  (bind _output_get (Variables|Default|SetactorTagsToCheck (GameplayTags|MakeGameplayTagContainerfromArray _gameplaytags)))
  (return _output_get))

(fn OnRep_bDaySequenceModifierEnabled
  (if (Variables|Default|GetbDaySequenceModifierEnabled)
    (CallFunction|EnableDaysequenceVolume)
    (Development|PrintString "Enabled DayNightForest TOD Modifier" false)
    (else
      (CallFunction|DisableDaysequenceVolume)
      (Development|PrintString "Disabled DayNightForest TOD Modifier" false))))

(fn OnRep_bIsActive
  (if (Variables|Default|GetbIsActive)
    (CallFunction|ActivateDaytime)))

(fn ConstructionScript)
