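# Additive unlit material for distant night-forest lights, drawn as glowing, camera-facing flare sprites.
# Vertex math collapses quads to pivots, jitters them with a time-scrolled random texture, then rebuilds view-aligned sprites that shrink with camera proximity and small height difference.
# Random values weight flare-texture channels and blend two color parameters, passed through vertex interpolators.
# Brightness follows time of day, peaking at night, and fades softly near scene geometry.
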Material /FirePetalDayNightForest/Assets/Lighting/M_NightForest_DistantLights
graph {
 properties {bAutomaticallySetUsageInEditor=false, blendMode=BLEND_Additive, shadingModel=MSM_Unlit, usedShadingModels=Unlit}
 nodes [
  n1 = AppendVector(A=Add(A=TextureCoordinate()) {constB=-0.5}, B=Constant())
  n2 = Multiply(A=Transform(None=@n1)) {constB=-200}
  n3 = Add(A=Multiply(A=Time(), B=Constant2Vector() {g=0.0131, r=0.0314}), B=Multiply(A=TextureCoordinate() {coordinateIndex=1}) {constB=3.687})
  n4 = TextureSample(UVs=Multiply(A=@n3) {constB=0.136}) {samplerType=SAMPLERTYPE_LinearColor, texture=&/T_NightForest_DistantLights}
  n5 = Multiply(A=Add(A=@n4.RGB) {constB=-0.5}) {constB=5000}
  n6 = Multiply(A=@n1, B=Constant3Vector() {constant=<rgba 1 -1 1 0>})
  n7 = TextureSample(UVs=Multiply(A=@n3)) {samplerType=SAMPLERTYPE_LinearColor, texture=&T_NightForest_DistantLights}
  n8 = Multiply(A=@n7.A) {constB=3000}
  n9 = WorldPosition() {worldPositionShaderOffset=WPT_CameraRelative}
  n10 = Distance(A=Add(A=NamedRerouteUsage(), B=@n9.XYZ), B=Constant())
  n11 = Abs(None=@n9.Z)
  n12 = Multiply(A=Saturate(None=Divide(A=@n10) {constB=70000}), B=Saturate(None=Divide(A=@n11) {constB=6000}))
  Add(
   A=NamedRerouteDeclaration(None=Add(A=@n2, B=@n5)) {name=Local Pivots, nodeColor=<rgba 1 0.1875 0 1>}
   B=Multiply(A=Transform(None=@n6) {transformSourceType=TRANSFORMSOURCE_View}, B=Multiply(A=@n8, B=@n12)) {constB=400}
  )
  n13 = Power(Base=NamedRerouteUsage()) {constExponent=16}
  n14 = MaterialFunctionCall(Float3=@n13) {materialFunction=&BreakOutFloat3Components}
  n15 = Add(A=Add(A=@n14.R, B=@n14.R), B=@n14.B)
  n16 = MaterialFunctionCall(Float3=VertexInterpolator(VS=Divide(A=@n13, B=@n15)).PS) {materialFunction=&BreakOutFloat3Components}
  n17 = TextureSample() {texture=&/T_NightForest_DistantLights_Flare}
  n18 = Add(A=Add(A=Multiply(A=@n16.R, B=@n17.R), B=Multiply(A=@n16.G, B=@n17.G)), B=Multiply(A=@n16.B, B=@n17.B))
  n19 = Multiply(A=@n18, B=@n18)
  n20 = VectorParameter() {parameterName=Color2, defaultValue=<rgba 0.0777 0.0217 1 1>}
  n21 = VectorParameter() {parameterName=Color1, defaultValue=<rgba 0.947 0 2 1>}
  n22 = VertexInterpolator(VS=NamedRerouteUsage())
  n23 = Multiply(A=Multiply(A=Add(A=@n19), B=LinearInterpolate(A=@n20.RGB, B=@n21.RGB, Alpha=@n22.PS))) {constB=3}
  n24 = Constant() {r=0.5}
  n25 = Constant() {r=0.7}
  n26 = Constant() {r=1}
  n27 = SceneDepth()
  n28 = PixelDepth()
  n29 = Saturate(None=Divide(A=Subtract(A=@n27.R, B=@n28.R)) {constB=3000})
  Multiply(
   A=Multiply(
    A=Multiply(A=@n18, B=@n23)
    B=MaterialFunctionCall(Value_Morning=@n24, Value_Day=@n25, Value_Evening=@n24, Value_Night=@n26) {materialFunction=&Scalar_Time_of_Day}.Result
   )
   B=Multiply(A=@n29, B=@n29)
  )
  NamedRerouteDeclaration(None=@n7.RGB) {name=Random RGB, nodeColor=<rgba 1 0 0.9844 1>}
  NamedRerouteDeclaration(None=@n4.A) {name=Random Hue, nodeColor=<rgba 0.5234 1 0 1>}
 ]
}
